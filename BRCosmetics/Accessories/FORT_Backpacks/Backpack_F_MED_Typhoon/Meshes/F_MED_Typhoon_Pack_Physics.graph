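# Physics asset for the Typhoon backpack cosmetic, giving collision bodies and joint limits to the pack and its three cables.
# One large capsule covers the pack, and chains of small capsules follow cable A (five segments), B and C (four each).
# Constraints join neighboring segments within each cable, allowing moderate swing but little twist.
# No constraint links the pack body to the cables or one cable to another.
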PhysicsAsset /BRCosmetics/Accessories/FORT_Backpacks/Backpack_F_MED_Typhoon/Meshes/F_MED_Typhoon_Pack_Physics
bodies {
 cableA_1 [{shapeType=Capsule, length=2, radius=0.6, center=<1 0 0>, rotation=<rot -90 0 0>}]
 cableA_2 [{shapeType=Capsule, length=2, radius=0.6, center=<1 0 0>, rotation=<rot -90 0 0>}]
 cableA_3 [{shapeType=Capsule, length=2, radius=0.6, center=<1 0 0>, rotation=<rot -90 0 0>}]
 cableA_4 [{shapeType=Capsule, length=1, radius=0.6, center=<0.4096 0 0>, rotation=<rot -90 0 0>}]
 cableA_5 [{shapeType=Capsule, length=2, radius=0.6, center=<1 0 0>, rotation=<rot -90 0 0>}]
 cableB_1 [{shapeType=Capsule, length=2, radius=0.6, center=<1 0 0>, rotation=<rot -90 0 0>}]
 cableB_2 [{shapeType=Capsule, length=2, radius=0.6, center=<1 0 0>, rotation=<rot -90 0 0>}]
 cableB_3 [{shapeType=Capsule, length=2, radius=0.6, center=<1 0 0>, rotation=<rot -90 0 0>}]
 cableB_4 [{shapeType=Capsule, length=2, radius=0.6, center=<1 0 0>, rotation=<rot -90 0 0>}]
 cableC_1 [{shapeType=Capsule, length=2, radius=0.6, center=<1 0 0>, rotation=<rot -90 0 0>}]
 cableC_2 [{shapeType=Capsule, length=2, radius=0.6, center=<1 0 0>, rotation=<rot -90 0 0>}]
 cableC_3 [{shapeType=Capsule, length=2, radius=0.6, center=<1 0 0>, rotation=<rot -90 0 0>}]
 cableC_4 [{shapeType=Capsule, length=2, radius=0.6, center=<1 0 0>, rotation=<rot -90 0 0>}]
 pack [{shapeType=Capsule, length=43, radius=14.5, center=<-5.4219 14.8859 -15.4198>, rotation=<rot 0 0 -4.7911>}]
}
constraints [
 {bone1Name=cableA_5, bone2Name=cableA_4, swing1LimitDegrees=45, swing2LimitDegrees=45, twistLimitDegrees=5}
 {bone1Name=cableA_2, bone2Name=cableA_1, swing1LimitDegrees=45, swing2LimitDegrees=45, twistLimitDegrees=5}
 {bone1Name=cableA_3, bone2Name=cableA_2, swing1LimitDegrees=45, swing2LimitDegrees=45, twistLimitDegrees=5}
 {bone1Name=cableA_4, bone2Name=cableA_3, swing1LimitDegrees=45, swing2LimitDegrees=45, twistLimitDegrees=5}
 {bone1Name=cableC_2, bone2Name=cableC_1, swing1LimitDegrees=45, swing2LimitDegrees=45, twistLimitDegrees=5}
 {bone1Name=cableC_3, bone2Name=cableC_2, swing1LimitDegrees=45, swing2LimitDegrees=45, twistLimitDegrees=5}
 {bone1Name=cableC_4, bone2Name=cableC_3, swing1LimitDegrees=45, swing2LimitDegrees=45, twistLimitDegrees=5}
 {bone1Name=cableB_2, bone2Name=cableB_1, swing1LimitDegrees=45, swing2LimitDegrees=45, twistLimitDegrees=5}
 {bone1Name=cableB_3, bone2Name=cableB_2, swing1LimitDegrees=45, swing2LimitDegrees=45, twistLimitDegrees=5}
 {bone1Name=cableB_4, bone2Name=cableB_3, swing1LimitDegrees=45, swing2LimitDegrees=45, twistLimitDegrees=5}
]
