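# Animation Blueprint for the Typhoon backpack cosmetic, animating the pack through idle, skydive and thumbs-up reactions to enemy eliminations.
# A state machine also enters notify-randomized idle breaks, though all three variants play the idle sequence.
# Its pose gets an additive pack roll while skydiving, then chained AnimDynamics on three cable chains and a RigidBody pass.
# A chest-opened flag is set and cleared but drives nothing.
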AnimBlueprint /BRCosmetics/Accessories/FORT_Backpacks/Backpack_F_MED_Typhoon/Animations/F_MED_Typhoon_Pack_AnimBP
target_skeleton=/BRCosmetics/Accessories/FORT_Backpacks/Backpack_F_MED_Typhoon/Meshes/F_MED_Typhoon_Pack_Skeleton
graphs [
 AnimGraph {nodes [
   Input Pose()
   AnimGraphNode_StateMachine_0 = State Machine() {is=title AnimGraphNode_StateMachine}
   n1 = "Transform (Modify) Bone - Bone: pack"(
    ComponentPose=Local To Component(LocalPose=@AnimGraphNode_StateMachine_0)
    Alpha=Get Is Sky Diving() -> (Is Sky Diving:double)
    Translation="0,0,0"
    Rotation="0,0,20"
   ) {
    rotationMode=BMM_Additive
    alphaBoolBlend {blendOption=Linear}
    alphaScaleBiasClamp {bInterpResult=true, interpSpeedIncreasing=20}
    boneToModify {boneName=pack}
    rotation {roll=20}
   }
   n2 = Anim Dynamics - Bone: cableC_1(ComponentPose=@n1, Alpha=1, ExternalForce="0,0,0") {
    angularDampingOverride=1
    angularSpringConstant=1500
    bAngularSpring=true
    bGravityOverrideInSimSpace=true
    bOverrideAngularDamping=true
    gravityScale=0.35
    simulationSpace=RootRelative
    alphaBoolBlend {blendOption=Linear}
    boundBone {boneName=cableC_1}
    componentLinearAccScale=<0.3 0.3 0.3>
    componentLinearVelScale=<0.3 0.3 0.3>
    physicsBodyDefinitions [{
      sphereCollisionRadius=0
      boundBone {boneName=cableC_1}
      boxExtents=<10 20 20>
      constraintSetup {angularLimitsMax=<5 5 0>, angularLimitsMin=<-5 -5 -25>, angularTarget=<0 0 0>}
      localJointOffset=<5 0 0>
     }]
   }
   n3 = Anim Dynamics - Bone: cableC_2(ComponentPose=@n2, Alpha=1, ExternalForce="0,0,0") {
    angularDampingOverride=1
    angularSpringConstant=1500
    bAngularSpring=true
    bGravityOverrideInSimSpace=true
    bOverrideAngularDamping=true
    gravityScale=0.35
    simulationSpace=RootRelative
    alphaBoolBlend {blendOption=Linear}
    boundBone {boneName=cableC_2}
    componentLinearAccScale=<0.3 0.3 0.3>
    componentLinearVelScale=<0.3 0.3 0.3>
    physicsBodyDefinitions [{
      sphereCollisionRadius=0
      boundBone {boneName=cableC_2}
      boxExtents=<10 20 20>
      constraintSetup {angularLimitsMax=<20 5 0>, angularLimitsMin=<-20 -5 -30>, angularTarget=<0 0 0>}
      localJointOffset=<5 0 0>
     }]
   }
   n4 = Anim Dynamics - Bone: cableC_3(ComponentPose=@n3, Alpha=1, ExternalForce="0,0,0") {
    angularDampingOverride=1
    angularSpringConstant=1500
    bAngularSpring=true
    bGravityOverrideInSimSpace=true
    bOverrideAngularDamping=true
    gravityScale=0.35
    simulationSpace=RootRelative
    alphaBoolBlend {blendOption=Linear}
    boundBone {boneName=cableC_3}
    componentLinearAccScale=<0.3 0.3 0.3>
    componentLinearVelScale=<0.3 0.3 0.3>
    physicsBodyDefinitions [{
      sphereCollisionRadius=0
      boundBone {boneName=cableC_3}
      boxExtents=<10 20 20>
      constraintSetup {angularLimitsMax=<5 5 0>, angularLimitsMin=<-5 -5 -25>, angularTarget=<0 0 0>}
      localJointOffset=<5 0 0>
     }]
   }
   n5 = Anim Dynamics - Bone: cableC_4(ComponentPose=@n4, Alpha=1, ExternalForce="0,0,0") {
    angularDampingOverride=1
    angularSpringConstant=1500
    bAngularSpring=true
    bGravityOverrideInSimSpace=true
    bOverrideAngularDamping=true
    gravityScale=0.35
    simulationSpace=RootRelative
    alphaBoolBlend {blendOption=Linear}
    boundBone {boneName=cableC_4}
    componentLinearAccScale=<0.3 0.3 0.3>
    componentLinearVelScale=<0.3 0.3 0.3>
    physicsBodyDefinitions [{
      sphereCollisionRadius=0
      boundBone {boneName=cableC_4}
      boxExtents=<10 20 20>
      constraintSetup {angularLimitsMax=<5 5 0>, angularLimitsMin=<-5 -5 -25>, angularTarget=<0 0 0>}
      localJointOffset=<5 0 0>
     }]
   }
   n6 = Anim Dynamics - Bone: cableB_1(ComponentPose=@n5, Alpha=1, ExternalForce="0,0,0") {
    angularDampingOverride=1
    angularSpringConstant=1500
    bAngularSpring=true
    bGravityOverrideInSimSpace=true
    bOverrideAngularDamping=true
    gravityScale=0.35
    simulationSpace=RootRelative
    alphaBoolBlend {blendOption=Linear}
    boundBone {boneName=cableB_1}
    componentLinearAccScale=<0.3 0.3 0.3>
    componentLinearVelScale=<0.3 0.3 0.3>
    physicsBodyDefinitions [{
      sphereCollisionRadius=0
      boundBone {boneName=cableB_1}
      boxExtents=<10 20 20>
      constraintSetup {angularLimitsMax=<5 5 0>, angularLimitsMin=<-5 -5 -25>, angularTarget=<0 0 0>}
      localJointOffset=<5 0 0>
     }]
   }
   n7 = Anim Dynamics - Bone: cableB_2(ComponentPose=@n6, Alpha=1, ExternalForce="0,0,0") {
    angularDampingOverride=1
    angularSpringConstant=1500
    bAngularSpring=true
    bGravityOverrideInSimSpace=true
    bOverrideAngularDamping=true
    gravityScale=0.35
    simulationSpace=RootRelative
    alphaBoolBlend {blendOption=Linear}
    boundBone {boneName=cableB_2}
    componentLinearAccScale=<0.3 0.3 0.3>
    componentLinearVelScale=<0.3 0.3 0.3>
    physicsBodyDefinitions [{
      sphereCollisionRadius=0
      boundBone {boneName=cableB_2}
      boxExtents=<10 20 20>
      constraintSetup {angularLimitsMax=<5 5 0>, angularLimitsMin=<-5 -5 -25>, angularTarget=<0 0 0>}
      localJointOffset=<5 0 0>
     }]
   }
   n8 = Anim Dynamics - Bone: cableA_1(ComponentPose=@n7, Alpha=1, ExternalForce="0,0,0") {
    angularDampingOverride=1
    angularSpringConstant=1500
    bAngularSpring=true
    bGravityOverrideInSimSpace=true
    bOverrideAngularDamping=true
    gravityScale=0.35
    simulationSpace=RootRelative
    alphaBoolBlend {blendOption=Linear}
    boundBone {boneName=cableA_1}
    componentLinearAccScale=<0.3 0.3 0.3>
    componentLinearVelScale=<0.3 0.3 0.3>
    physicsBodyDefinitions [{
      sphereCollisionRadius=0
      boundBone {boneName=cableA_1}
      boxExtents=<10 20 20>
      constraintSetup {angularLimitsMax=<5 5 0>, angularLimitsMin=<-5 -5 -25>, angularTarget=<0 0 0>}
      localJointOffset=<0 0 0>
     }]
   }
   n9 = Anim Dynamics - Bone: cableA_2(ComponentPose=@n8, Alpha=1, ExternalForce="0,0,0") {
    angularDampingOverride=1
    angularSpringConstant=1500
    bAngularSpring=true
    bGravityOverrideInSimSpace=true
    bOverrideAngularDamping=true
    gravityScale=0.35
    simulationSpace=RootRelative
    alphaBoolBlend {blendOption=Linear}
    boundBone {boneName=cableA_2}
    componentLinearAccScale=<0.3 0.3 0.3>
    componentLinearVelScale=<0.3 0.3 0.3>
    physicsBodyDefinitions [{
      sphereCollisionRadius=0
      boundBone {boneName=cableA_2}
      boxExtents=<10 20 20>
      constraintSetup {angularLimitsMax=<5 5 0>, angularLimitsMin=<-5 -5 -25>, angularTarget=<0 0 0>}
      localJointOffset=<0 0 0>
     }]
   }
   n10 = Anim Dynamics - Bone: cableA_3(ComponentPose=@n9, Alpha=1, ExternalForce="0,0,0") {
    angularDampingOverride=1
    angularSpringConstant=1500
    bAngularSpring=true
    bGravityOverrideInSimSpace=true
    bOverrideAngularDamping=true
    gravityScale=0.35
    simulationSpace=RootRelative
    alphaBoolBlend {blendOption=Linear}
    boundBone {boneName=cableA_3}
    componentLinearAccScale=<0.3 0.3 0.3>
    componentLinearVelScale=<0.3 0.3 0.3>
    physicsBodyDefinitions [{
      sphereCollisionRadius=0
      boundBone {boneName=cableA_3}
      boxExtents=<10 20 20>
      constraintSetup {angularLimitsMax=<5 5 0>, angularLimitsMin=<-5 -5 -25>, angularTarget=<0 0 0>}
      localJointOffset=<0 0 0>
     }]
   }
   n11 = Anim Dynamics - Bone: cableA_4(ComponentPose=@n10, Alpha=1, ExternalForce="0,0,0") {
    angularDampingOverride=1
    angularSpringConstant=1500
    bAngularSpring=true
    bGravityOverrideInSimSpace=true
    bOverrideAngularDamping=true
    gravityScale=0.35
    simulationSpace=RootRelative
    alphaBoolBlend {blendOption=Linear}
    boundBone {boneName=cableA_4}
    componentLinearAccScale=<0.3 0.3 0.3>
    componentLinearVelScale=<0.3 0.3 0.3>
    physicsBodyDefinitions [{
      sphereCollisionRadius=0
      boundBone {boneName=cableA_4}
      boxExtents=<10 20 20>
      constraintSetup {angularLimitsMax=<5 5 0>, angularLimitsMin=<-5 -5 -25>, angularTarget=<0 0 0>}
      localJointOffset=<0 0 0>
     }]
   }
   n12 = Anim Dynamics - Bone: cableA_5(ComponentPose=@n11, Alpha=1, ExternalForce="0,0,0") {
    angularDampingOverride=1
    angularSpringConstant=1500
    bAngularSpring=true
    bGravityOverrideInSimSpace=true
    bOverrideAngularDamping=true
    gravityScale=0.35
    simulationSpace=RootRelative
    alphaBoolBlend {blendOption=Linear}
    boundBone {boneName=cableA_5}
    componentLinearAccScale=<0.3 0.3 0.3>
    componentLinearVelScale=<0.3 0.3 0.3>
    physicsBodyDefinitions [{
      sphereCollisionRadius=0
      boundBone {boneName=cableA_5}
      boxExtents=<10 20 20>
      constraintSetup {angularLimitsMax=<5 5 0>, angularLimitsMin=<-5 -5 -25>, angularTarget=<0 0 0>}
      localJointOffset=<0 0 0>
     }]
   }
   n13 = RigidBody(ComponentPose=@n12) {
    bClampLinearTranslationLimitToRefPose=true
    bOverrideWorldGravity=true
    overridePhysicsAsset=/BRCosmetics/Accessories/FORT_Backpacks/Backpack_F_MED_Typhoon/Meshes/F_MED_Typhoon_Pack_Physics
    simulationSpace=BaseBoneSpace
    alphaBoolBlend {blendOption=Linear}
    baseBoneRef {boneName=root}
    componentLinearAccScale=<0.25 0.25 0.35>
    componentLinearVelScale=<0.25 0.25 0.35>
    overrideWorldGravity {z=-980}
   }
   Output Pose(Result=Component To Local(ComponentPose=@n13)) {layerGroup=None}
  ]}
 State Machine {
  entry=Locomotion
  path=AnimGraph.AnimGraphNode_StateMachine_0.State Machine
  ids=id
  nodes [
   AnimStateNode_0 = Locomotion() {
    is=title AnimStateNode
    stateEntered {guid=#1, notifyName=Entered_RandomLocoBreak}
    stateFullyBlended {guid=#2}
    stateLeft {guid=#3}
   }
   AnimStateNode_1 = Skydiving() {
    is=title AnimStateNode
    stateEntered {guid=#4}
    stateFullyBlended {guid=#5}
    stateLeft {guid=#6}
   }
   AnimStateNode_4 = Wink() {is=title AnimStateNode}
   AnimStateNode_6 = IdleBreak() {
    is=title AnimStateNode
    stateEntered {guid=#7}
    stateFullyBlended {guid=#8}
    stateLeft {guid=#9}
   }
   AnimStateTransitionNode_1 = AnimStateTransitionNode()
   AnimStateTransitionNode_3 = AnimStateTransitionNode()
   AnimStateTransitionNode_4 = AnimStateTransitionNode()
   AnimStateTransitionNode_5 = AnimStateTransitionNode() {bAutomaticRuleBasedOnSequencePlayerInState=true}
   AnimStateTransitionNode_6 = AnimStateTransitionNode() {
    transitionEnd {guid=#10}
    transitionInterrupt {guid=#11}
    transitionStart {guid=#12}
   }
   AnimStateTransitionNode_9 = AnimStateTransitionNode() {
    bAutomaticRuleBasedOnSequencePlayerInState=true
    transitionEnd {guid=#13}
    transitionInterrupt {guid=#14}
    transitionStart {guid=#15}
   }
  ]
  links [
   {
    dst=Locomotion
    rule=AnimGraph.AnimGraphNode_StateMachine_0.State Machine.AnimStateTransitionNode_9.Transition
    src=IdleBreak
    via=AnimStateTransitionNode_9
   }
   {
    dst=IdleBreak
    rule=AnimGraph.AnimGraphNode_StateMachine_0.State Machine.AnimStateTransitionNode_6.Transition
    src=Locomotion
    via=AnimStateTransitionNode_6
   }
   {
    dst=Skydiving
    rule=AnimGraph.AnimGraphNode_StateMachine_0.State Machine.AnimStateTransitionNode_3.Transition
    src=Locomotion
    via=AnimStateTransitionNode_3
   }
   {
    dst=Wink
    rule=AnimGraph.AnimGraphNode_StateMachine_0.State Machine.AnimStateTransitionNode_1.Transition
    src=Locomotion
    via=AnimStateTransitionNode_1
   }
   {
    dst=Locomotion
    rule=AnimGraph.AnimGraphNode_StateMachine_0.State Machine.AnimStateTransitionNode_4.Transition
    src=Skydiving
    via=AnimStateTransitionNode_4
   }
   {
    dst=Locomotion
    rule=AnimGraph.AnimGraphNode_StateMachine_0.State Machine.AnimStateTransitionNode_5.Transition
    src=Wink
    via=AnimStateTransitionNode_5
   }
  ]
 }
 Locomotion {
  path=AnimGraph.AnimGraphNode_StateMachine_0.State Machine.AnimStateNode_0.Locomotion
  nodes [Output Animation Pose(Result=Sequence Player 'Typhoon_Backpack_Idle'() {sequence=&/Typhoon_Backpack_Idle})]
 }
 Skydiving {
  path=AnimGraph.AnimGraphNode_StateMachine_0.State Machine.AnimStateNode_1.Skydiving
  nodes [Output Animation Pose(Result=Sequence Player 'Typhoon_Backpack_Skydive'() {sequence=&/Typhoon_Backpack_Skydive})]
 }
 Wink {
  path=AnimGraph.AnimGraphNode_StateMachine_0.State Machine.AnimStateNode_4.Wink
  nodes [Output Animation Pose(Result=Sequence Player 'Typhoon_Backpack_ThumbsUp'() {sequence=&/Typhoon_Backpack_ThumbsUp})]
 }
 IdleBreak {
  path=AnimGraph.AnimGraphNode_StateMachine_0.State Machine.AnimStateNode_6.IdleBreak
  nodes [
   n1 = Blend Poses by int!(
    BlendPose_0:Pose Link Structure=Sequence Player 'Typhoon_Backpack_Idle'!(bLoopAnimation:bool=False) -> (Pose) {bLoopAnimation=false, sequence=&Typhoon_Backpack_Idle}
    BlendPose_1:Pose Link Structure=Sequence Player 'Typhoon_Backpack_Idle'(bLoopAnimation=False) {bLoopAnimation=false, sequence=&Typhoon_Backpack_Idle}
    BlendPose_2:Pose Link Structure=Sequence Player 'Typhoon_Backpack_Idle'(bLoopAnimation=False) {bLoopAnimation=false, sequence=&Typhoon_Backpack_Idle}
    BlendTime_0:float=0.1
    BlendTime_1:float=0.1
    BlendTime_2:float=0.1
    ActiveChildIndex:int=Get Choose_LocoBreak_Index() -> (Choose_LocoBreak_Index:int=0)
   ) -> (Pose:Pose Link Structure) {
    is=title AnimGraphNode_BlendListByInt
    blendType=Linear
    blendPose [{linkId=-1, sourceLinkId=-1}, {linkId=-1, sourceLinkId=-1}, {linkId=-1, sourceLinkId=-1}]
    blendTime [0.1,0.1,0.1]
   }
   Output Animation Pose(Result=@n1)
  ]
 }
 Transition {
  path=AnimGraph.AnimGraphNode_StateMachine_0.State Machine.AnimStateTransitionNode_1.Transition
  nodes [Result(bCanEnterTransition=Get bEnemyPlayerEliminated() -> (bEnemyPlayerEliminated:bool))]
 }
 Transition {
  path=AnimGraph.AnimGraphNode_StateMachine_0.State Machine.AnimStateTransitionNode_3.Transition
  nodes [Result(bCanEnterTransition=Get bIsSkyDiving())]
 }
 Transition {
  path=AnimGraph.AnimGraphNode_StateMachine_0.State Machine.AnimStateTransitionNode_4.Transition
  nodes [Result(bCanEnterTransition=NOT Boolean(A=Get bIsSkyDiving()))]
 }
 Transition {path=AnimGraph.AnimGraphNode_StateMachine_0.State Machine.AnimStateTransitionNode_5.Transition, nodes [Result()]}
 Transition {
  path=AnimGraph.AnimGraphNode_StateMachine_0.State Machine.AnimStateTransitionNode_6.Transition
  nodes [Result(bCanEnterTransition=Get AllowLocomotionBreak() -> (AllowLocomotionBreak:bool))]
 }
 Transition {
  path=AnimGraph.AnimGraphNode_StateMachine_0.State Machine.AnimStateTransitionNode_9.Transition
  nodes [
   Result(bCanEnterTransition=NOT Boolean(A=Get AllowLocomotionBreak()))
   Get IsLocoBreaking() -> (IsLocoBreaking:bool)
  ]
 }
 EventGraph {nodes [
   Event BlueprintUpdateAnimation()
   > Set Is Sky Diving(Is Sky Diving:double="To Float (Boolean)"(InBool=Get bIsSkyDiving())) -> (Output_Get:double)
   n1 = Event BlueprintProcessAnimInputEvent()
   > n2 = Branch(Condition=IsNamed(self=@n1.AnimInputEvent, InEventName=aie.EnemyPlayerEliminated))
   > Set bEnemyPlayerEliminated(bEnemyPlayerEliminated:bool=true) -> (Output_Get:bool)
   Branch(execute=@n2.else, Condition=IsNamed(self=@n1.AnimInputEvent, InEventName=aie.ChestOpened))
   > Set bChestOpened(bChestOpened:bool=true) -> (Output_Get:bool)
   AnimNotify_TriggerLocomotionRandomizer() {is=title K2Node_Event}
   > n3 = Set AllowLocomotionBreak(AllowLocomotionBreak:bool=RandomBoolWithWeight!(Weight:float=0.3) -> (ReturnValue:bool)) -> (Output_Get:bool)
   > Branch(Condition=@n3)
   > Set Choose_LocoBreak_Index(Choose_LocoBreak_Index:int=RandomInteger!(Max:int=3) -> (ReturnValue:int=0)) -> (Output_Get:int=0)
   Event BlueprintPostEvaluateAnimation() {is=title K2Node_Event}
   > Set bEnemyPlayerEliminated(bEnemyPlayerEliminated=false)
   > Set bChestOpened(bChestOpened=false)
   AnimNotify_Entered_RandomLocoBreak() {is=title K2Node_Event}
   > Set AllowLocomotionBreak(AllowLocomotionBreak=false)
   AnimNotify_IsLocoBreaking() {is=title K2Node_Event}
   > Set IsLocoBreaking(IsLocoBreaking:bool=false) -> (Output_Get:bool)
   TryGetPawnOwner()
  ]}
]
variables [
 AllowLocomotionBreak {type=boolean}
 bChestOpened {type=boolean}
 bEnemyPlayerEliminated {type=boolean}
 Choose_LocoBreak_Index {type=integer}
 Is Sky Diving {type=number}
 IsLocoBreaking {type=boolean}
]
