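# Looping Niagara effect for the female Typhoon backpack cosmetic, producing blue electrical arcs with spark and smoke layers.
# Lightning emitters at fixed offsets flash meshes with lights on randomized loop delays, animate material parameters and emit location events; one is disabled.
# Marks_Lightning_Loop scatters tiny flickering arc meshes by skeletal mesh sampling, previewed on the pack mesh.
# Spark sprites fall under drag; eroding SubUV smoke has its only spawn burst disabled.
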NiagaraSystem /BRCosmetics/Accessories/FORT_Backpacks/Backpack_F_MED_Typhoon/FX/NS_Backpack_TyphoonFemale_Loop
dependencies {
 usedDataInterfaces [&NiagaraDataInterfaceColorCurve, &NiagaraDataInterfaceCurve]
 usedDynamicInputs []
 usedModules [/Niagara/Modules/System/SystemLifeCycle, &SystemState]
}
emitters {
 Left_eye_lightning {
  bEnabled=false
  emitterUpdateScript [
   &EmitterState(
    Life Cycle Mode=::NewEnumerator1
    Loop Delay=&UniformRangedFloat(Minimum=1, Maximum=2)
    Loop Duration=&UniformRangedFloat(Minimum=0.5, Maximum=1)
    Recalculate Duration Each Loop=true
    UseLoopDelay=true
   )
   &SpawnRate(Use Spawn Probability=false)
  ]
  particleSpawnScript [
   /Niagara/Modules/Spawn/Initialization/InitializeParticle(
    Lifetime=&UniformRangedFloat(Minimum=0.15, Maximum=0.15)
    Mesh Scale=<0.15 0.3 0.15>
    Module.Write Color=true
    Module.Write Lifetime=true
    Module.Write Mass=true
    Module.Write Material Random=true
    Module.Write Position=true
    Module.Write Scale=true
    Module.Write SpriteRotation=false
    Module.Write SpriteSize=false
    Position=&SimulationPosition()
    Sprite Size=<25 25>
   )
   &InitialMeshOrientation(Orientation Vector=<10 50 -20>, Rotation=<0 -0.5 -0.5>, Use Orientation Vector=true, Use Rotation Vector=true) {module=InitialMeshOrientation001}
   /Niagara/Modules/Spawn/Location/SphereLocation(MaskSpawnGroup=false, Module.Override Randomness=false, Offset=<-9.4 -1.5 -25>, Sphere Radius=0, Surface Only=false)
  ]
  particleUpdateScript [
   &ParticleState()
   &ScaleColor(
    Scale RGB=&MakeVectorFromLinearColorRGB(Color=&ColorFromCurve(ColorCurve=DataInterface() {propertyValues {
        AlphaCurve {keys [<at 0 1 RCIM_Cubic>,<at 1.0315 1 RCIM_Cubic>]}
        BlueCurve {keys [<at 0 100 RCIM_Cubic>]}
        GreenCurve {keys [<at 0 8.6868 RCIM_Cubic>]}
        RedCurve {keys [<at 0 3.9455 RCIM_Cubic>]}
       }}))
    ScaleA=true
    ScaleRGB=true
   )
   &DynamicMaterialParameters(
    Index 0 Param 1=&UniformRangedFloat(Minimum=4, Maximum=7)
    Index 0 Param 2=&FloatFromCurve(FloatCurve=DataInterface() {propertyValues {Curve {keys [<at 0.1129 1>,<at 1 -1>]}}})
    Index 0 Param 3=&UniformRangedFloat(Minimum=0.9, Maximum=1.5)
    Module.Param0WriteEnabled=true
    Module.Param1WriteEnabled=true
    Module.Param2WriteEnabled=true
    Module.Param3WriteEnabled=true
   )
   &SolveForcesAndVelocity(Clamp Velocity=false, Limit Acceleration=false)
   /SetVariables(Particles.Light_Size=10) {isSetParameters=true, module=SetVariables_#1}
   /SetVariables(Particles.Light_Color:LinearColor=<rgba 0.2898 0.4518 4 1>) {isSetParameters=true, module=SetVariables_#2}
   &GenerateLocationEvent(Event Send Rate=30, Module.Use Event Send Rate=true)
  ]
  rendererClasses [&NiagaraLightRendererProperties, &NiagaraMeshRendererProperties]
  renderers [
   {rendererClass=&NiagaraLightRendererProperties, rendererIndex=0}
   {rendererClass=&NiagaraMeshRendererProperties, rendererIndex=1}
  ]
 }
 Left_eye_lightning001 {
  emitterUpdateScript [
   &EmitterState(
    Life Cycle Mode=::NewEnumerator1
    Loop Delay=&UniformRangedFloat(Minimum=1, Maximum=3)
    Loop Duration=&UniformRangedFloat(Minimum=0.5, Maximum=1)
    Recalculate Duration Each Loop=true
    UseLoopDelay=true
   )
   &SpawnRate(Use Spawn Probability=false)
  ]
  particleSpawnScript [
   &InitializeParticle(
    Lifetime=&UniformRangedFloat(Minimum=0.15, Maximum=0.15)
    Mesh Scale=&Multiply_Vector(A=<0.8 0.8 0.8>, B=<0.8 0.8 0.8>)
    Module.Write Color=true
    Module.Write Lifetime=true
    Module.Write Mass=true
    Module.Write Material Random=true
    Module.Write Position=true
    Module.Write Scale=true
    Module.Write SpriteRotation=false
    Module.Write SpriteSize=false
    Position=&SimulationPosition()
    Sprite Size=<25 25>
   )
   &InitialMeshOrientation(Orientation Vector=<10 50 -20>, Rotation=<0 -0.3 -0.5>, Use Orientation Vector=true, Use Rotation Vector=true) {module=InitialMeshOrientation001}
   &SphereLocation(MaskSpawnGroup=false, Module.Override Randomness=false, Offset=<1 -5 -2>, Sphere Radius=0, Surface Only=false)
  ]
  particleUpdateScript [
   &ParticleState()
   &ScaleColor(
    Scale RGB=&MakeVectorFromLinearColorRGB(Color=&ColorFromCurve(ColorCurve=DataInterface() {propertyValues {
        AlphaCurve {keys [<at 0 1 RCIM_Cubic>,<at 1.0315 1 RCIM_Cubic>]}
        BlueCurve {keys [<at 0 100 RCIM_Cubic>]}
        GreenCurve {keys [<at 0 8.6868 RCIM_Cubic>]}
        RedCurve {keys [<at 0 3.9455 RCIM_Cubic>]}
       }}))
    ScaleA=true
    ScaleRGB=true
   )
   &DynamicMaterialParameters(
    Index 0 Param 1=&UniformRangedFloat(Minimum=4, Maximum=7)
    Index 0 Param 2=&FloatFromCurve(FloatCurve=DataInterface() {propertyValues {Curve {keys [<at 0.1129 1>,<at 1 -1>]}}})
    Index 0 Param 3=&UniformRangedFloat(Minimum=0.9, Maximum=1.5)
    Module.Param0WriteEnabled=true
    Module.Param1WriteEnabled=true
    Module.Param2WriteEnabled=true
    Module.Param3WriteEnabled=true
   )
   &SolveForcesAndVelocity(Clamp Velocity=false, Limit Acceleration=false)
   /SetVariables(Particles.Light_Size=10) {isSetParameters=true, module=SetVariables_#1}
   /SetVariables(Particles.Light_Color:LinearColor=<rgba 0.5071 0.7906 7 1>) {isSetParameters=true, module=SetVariables_#2}
   &GenerateLocationEvent(Event Send Rate=30, Module.Use Event Send Rate=true)
  ]
  rendererClasses [&NiagaraLightRendererProperties, &NiagaraMeshRendererProperties]
  renderers [
   {rendererClass=&NiagaraLightRendererProperties, rendererIndex=0}
   {rendererClass=&NiagaraMeshRendererProperties, rendererIndex=1}
  ]
 }
 Left_eye_lightning002 {
  emitterUpdateScript [
   &EmitterState(
    Life Cycle Mode=::NewEnumerator1
    Loop Delay=&UniformRangedFloat(Minimum=2, Maximum=3)
    Loop Duration=&UniformRangedFloat(Minimum=1, Maximum=1)
    Recalculate Duration Each Loop=true
    UseLoopDelay=true
   )
   &SpawnRate(Use Spawn Probability=false)
  ]
  particleSpawnScript [
   &InitializeParticle(
    Lifetime=&UniformRangedFloat(Minimum=0.15, Maximum=0.15)
    Mesh Scale=&Multiply_Vector(A=<0.8 0.8 0.8>, B=<0.8 0.8 0.8>)
    Module.Write Color=true
    Module.Write Lifetime=true
    Module.Write Mass=true
    Module.Write Material Random=true
    Module.Write Position=true
    Module.Write Scale=true
    Module.Write SpriteRotation=false
    Module.Write SpriteSize=false
    Position=&SimulationPosition()
    Sprite Size=<25 25>
   )
   &InitialMeshOrientation(Orientation Vector=<10 50 -20>, Rotation=<0 0 0.5>, Use Orientation Vector=true, Use Rotation Vector=true) {module=InitialMeshOrientation001}
   &SphereLocation(MaskSpawnGroup=false, Module.Override Randomness=false, Offset=<5 0 0.8>, Sphere Radius=0, Surface Only=false)
  ]
  particleUpdateScript [
   &ParticleState()
   &ScaleColor(
    Scale RGB=&MakeVectorFromLinearColorRGB(Color=&ColorFromCurve(ColorCurve=DataInterface() {propertyValues {
        AlphaCurve {keys [<at 0 1 RCIM_Cubic>,<at 1.0315 1 RCIM_Cubic>]}
        BlueCurve {keys [<at 0 100 RCIM_Cubic>]}
        GreenCurve {keys [<at 0 8.6868 RCIM_Cubic>]}
        RedCurve {keys [<at 0 3.9455 RCIM_Cubic>]}
       }}))
    ScaleA=true
    ScaleRGB=true
   )
   &DynamicMaterialParameters(
    Index 0 Param 1=&UniformRangedFloat(Minimum=4, Maximum=7)
    Index 0 Param 2=&FloatFromCurve(FloatCurve=DataInterface() {propertyValues {Curve {keys [<at 0.1129 1>,<at 1 -1>]}}})
    Index 0 Param 3=&UniformRangedFloat(Minimum=0.9, Maximum=1.5)
    Module.Param0WriteEnabled=true
    Module.Param1WriteEnabled=true
    Module.Param2WriteEnabled=true
    Module.Param3WriteEnabled=true
   )
   &SolveForcesAndVelocity(Clamp Velocity=false, Limit Acceleration=false)
   /SetVariables(Particles.Light_Size=10) {isSetParameters=true, module=SetVariables_#1}
   /SetVariables(Particles.Light_Color:LinearColor=<rgba 0.5071 0.7906 7 1>) {isSetParameters=true, module=SetVariables_#2}
   &GenerateLocationEvent(Event Send Rate=30, Module.Use Event Send Rate=true)
  ]
  rendererClasses [&NiagaraLightRendererProperties, &NiagaraMeshRendererProperties]
  renderers [
   {rendererClass=&NiagaraLightRendererProperties, rendererIndex=0}
   {rendererClass=&NiagaraMeshRendererProperties, rendererIndex=1}
  ]
 }
 Left_eye_lightning003 {
  emitterUpdateScript [
   &EmitterState(
    Life Cycle Mode=::NewEnumerator1
    Loop Delay=&UniformRangedFloat(Minimum=2, Maximum=3)
    Loop Duration=&UniformRangedFloat(Minimum=0.4, Maximum=1)
    Recalculate Duration Each Loop=true
    UseLoopDelay=true
   )
   &SpawnRate(Use Spawn Probability=false)
  ]
  particleSpawnScript [
   &InitializeParticle(
    Lifetime=&UniformRangedFloat(Minimum=0.1, Maximum=0.15)
    Mesh Scale=<0.8 1 0.8>
    Module.Write Color=true
    Module.Write Lifetime=true
    Module.Write Mass=true
    Module.Write Material Random=true
    Module.Write Position=true
    Module.Write Scale=true
    Module.Write SpriteRotation=false
    Module.Write SpriteSize=false
    Position=&SimulationPosition()
    Sprite Size=<25 25>
   )
   &InitialMeshOrientation(Orientation Vector=<10 50 -20>, Rotation=<-0.1 0.2 -0.2>, Use Orientation Vector=true, Use Rotation Vector=true) {module=InitialMeshOrientation001}
   &SphereLocation(MaskSpawnGroup=false, Module.Override Randomness=false, Offset=<0 -3.2 -10>, Sphere Radius=0, Surface Only=false)
  ]
  particleUpdateScript [
   &ParticleState()
   &ScaleColor(
    Scale RGB=&MakeVectorFromLinearColorRGB(Color=&ColorFromCurve(ColorCurve=DataInterface() {propertyValues {
        AlphaCurve {keys [<at 0 1 RCIM_Cubic>,<at 1.0315 1 RCIM_Cubic>]}
        BlueCurve {keys [<at 0 100 RCIM_Cubic>]}
        GreenCurve {keys [<at 0 8.6868 RCIM_Cubic>]}
        RedCurve {keys [<at 0 3.9455 RCIM_Cubic>]}
       }}))
    ScaleA=true
    ScaleRGB=true
   )
   &DynamicMaterialParameters(
    Index 0 Param 1=&UniformRangedFloat(Minimum=4, Maximum=7)
    Index 0 Param 2=&FloatFromCurve(FloatCurve=DataInterface() {propertyValues {Curve {keys [<at 0.1129 1>,<at 1 -1>]}}})
    Index 0 Param 3=&UniformRangedFloat(Minimum=0.9, Maximum=1.5)
    Module.Param0WriteEnabled=true
    Module.Param1WriteEnabled=true
    Module.Param2WriteEnabled=true
    Module.Param3WriteEnabled=true
   )
   &SolveForcesAndVelocity(Clamp Velocity=false, Limit Acceleration=false)
   /SetVariables(Particles.Light_Size=10) {isSetParameters=true, module=SetVariables_#1}
   /SetVariables(Particles.Light_Color:LinearColor=<rgba 0.5071 0.7906 7 1>) {isSetParameters=true, module=SetVariables_#2}
   &GenerateLocationEvent(Event Send Rate=50, Module.Use Event Send Rate=true)
  ]
  rendererClasses [&NiagaraLightRendererProperties, &NiagaraMeshRendererProperties]
  renderers [
   {rendererClass=&NiagaraLightRendererProperties, rendererIndex=0}
   {rendererClass=&NiagaraMeshRendererProperties, rendererIndex=1}
  ]
 }
 Left_eye_lightning004 {
  emitterUpdateScript [
   &EmitterState(
    Life Cycle Mode=::NewEnumerator1
    Loop Delay=&UniformRangedFloat(Minimum=1, Maximum=2)
    Loop Duration=&UniformRangedFloat(Minimum=0.5, Maximum=1)
    Recalculate Duration Each Loop=true
    UseLoopDelay=true
   )
   &SpawnRate(Use Spawn Probability=false)
  ]
  particleSpawnScript [
   &InitializeParticle(
    Lifetime=&UniformRangedFloat(Minimum=0.15, Maximum=0.15)
    Mesh Scale=<0.2 0.3 0.2>
    Module.Write Color=true
    Module.Write Lifetime=true
    Module.Write Mass=true
    Module.Write Material Random=true
    Module.Write Position=true
    Module.Write Scale=true
    Module.Write SpriteRotation=false
    Module.Write SpriteSize=false
    Position=&SimulationPosition()
    Sprite Size=<25 25>
   )
   &InitialMeshOrientation(Orientation Vector=<1 1 0.1>, Rotation=<0.25 0 0>, Use Orientation Vector=true, Use Rotation Vector=true) {module=InitialMeshOrientation001}
   &SphereLocation(MaskSpawnGroup=false, Module.Override Randomness=false, Offset=<-6 -1.5 -28>, Sphere Radius=0, Surface Only=false)
  ]
  particleUpdateScript [
   &ParticleState()
   &ScaleColor(
    Scale RGB=&MakeVectorFromLinearColorRGB(Color=&ColorFromCurve(ColorCurve=DataInterface() {propertyValues {
        AlphaCurve {keys [<at 0 1 RCIM_Cubic>,<at 1.0315 1 RCIM_Cubic>]}
        BlueCurve {keys [<at 0 100 RCIM_Cubic>]}
        GreenCurve {keys [<at 0 8.6868 RCIM_Cubic>]}
        RedCurve {keys [<at 0 3.9455 RCIM_Cubic>]}
       }}))
    ScaleA=true
    ScaleRGB=true
   )
   &DynamicMaterialParameters(
    Index 0 Param 1=&UniformRangedFloat(Minimum=4, Maximum=7)
    Index 0 Param 2=&FloatFromCurve(FloatCurve=DataInterface() {propertyValues {Curve {keys [<at 0.1129 1>,<at 1 -1>]}}})
    Index 0 Param 3=&UniformRangedFloat(Minimum=0.9, Maximum=1.5)
    Module.Param0WriteEnabled=true
    Module.Param1WriteEnabled=true
    Module.Param2WriteEnabled=true
    Module.Param3WriteEnabled=true
   )
   &SolveForcesAndVelocity(Clamp Velocity=false, Limit Acceleration=false)
   /SetVariables(Particles.Light_Size=10) {isSetParameters=true, module=SetVariables_#1}
   /SetVariables(Particles.Light_Color:LinearColor=<rgba 0.2898 0.4518 4 1>) {isSetParameters=true, module=SetVariables_#2}
   &GenerateLocationEvent(Module.Use Event Send Rate=true)
  ]
  rendererClasses [&NiagaraLightRendererProperties, &NiagaraMeshRendererProperties]
  renderers [
   {rendererClass=&NiagaraLightRendererProperties, rendererIndex=0}
   {rendererClass=&NiagaraMeshRendererProperties, rendererIndex=1}
  ]
 }
 Marks_Lightning_Loop {
  emitterUpdateScript [
   &EmitterState(
    Life Cycle Mode=::NewEnumerator1
    Loop Delay=&UniformRangedFloat(Minimum=2, Maximum=3)
    Loop Duration=&UniformRangedFloat(Minimum=0.2, Maximum=0.7)
    Recalculate Duration Each Loop=true
    UseLoopDelay=true
   )
   &SpawnRate(SpawnRate=6, Use Spawn Probability=false)
   &SpawnBurst_Instantaneous(Spawn Count=2, Use Spawn Probability=false)
  ]
  particleSpawnScript [
   &InitializeParticle(
    Lifetime=&UniformRangedFloat(Minimum=0.07, Maximum=0.08)
    Mesh Scale=&Multiply_Vector(
     A=&UniformRangedVector(Minimum=<0.3 0.4 0.4>, Maximum=<0.4 0.5 0.5>)
     B=&VectorFromFloat(Value=0.1)
    )
    Module.Write Color=true
    Module.Write Lifetime=true
    Module.Write Mass=true
    Module.Write Material Random=true
    Module.Write Position=true
    Module.Write Scale=true
    Module.Write SpriteRotation=false
    Module.Write SpriteSize=false
    Position=&SimulationPosition()
    Sprite Size=<25 25>
   )
   &InitialMeshOrientation(
    Rotation=&UniformRangedVector(Minimum=<-1 0 0>, Maximum=<1 0 0>)
    Use Orientation Vector=true
    Use Rotation Vector=true
   ) {module=InitialMeshOrientation001}
   &AddVelocityFromPoint(Module.Use Falloff Distance=false, Velocity Strength=0.01)
   &DynamicMaterialParameters(
    Index 0 Param 1=&UniformRangedFloat(Minimum=0.15, Maximum=0.18)
    Index 0 Param 2=&UniformRangedFloat(Minimum=0, Maximum=1)
    Index 0 Param 3=100
    Index 0 Param 4=5
    Module.Param0WriteEnabled=true
    Module.Param1WriteEnabled=true
    Module.Param2WriteEnabled=true
    Module.Param3WriteEnabled=true
   )
   &SkeletalMeshLocation(
    Bone Sampling Mode=::NewEnumerator0
    Inherit Velocity=false
    Interpolate Mesh Sampling=false
    MaskSpawnGroup=false
    Module.Override Randomness=false
    Sampled Orientation Coordinate Space=::Local
    Sampled Position Coordinate Space=::Local
    Sampled Position Offset=&UniformRangedVector(Minimum=<2 2 2>, Maximum=<-4 -2 -2>)
    Sampled Position Offset Coordinate Space:ENiagaraMeshTransforms=::NewEnumerator3
    Skeletal Mesh=DataInterface() {propertyValues {
      DefaultMesh=null
      ExcludeBoneName=root
      OptionalComponentClass=null
      PreviewMesh=/BRCosmetics/Accessories/FORT_Backpacks/Backpack_F_MED_Typhoon/Meshes/F_MED_Typhoon_Pack
      SkinningMode=SkinOnTheFly
      SoftSourceActor=null
      SourceMode=Default
      UvSetIndex=0
      WholeMeshLOD=-1
      bExcludeBone=false
      bReadDeformedGeometry=true
      bRequireCurrentFrameData=true
      ComponentTags []
      FilteredBones [lowerarm_twist_01_l, lowerarm_twist_02_l]
      FilteredSockets []
      MeshUserParameter {parameter {name=None, typeDefHandle {registeredTypeIndex=1}, varData []}}
      SamplingRegions []
     }}
    Use Sampled Mesh Attribute as Ribbon ID=false
   )
  ]
  particleUpdateScript [
   &SolveForcesAndVelocity(Clamp Velocity=false, Limit Acceleration=false)
   &ParticleState()
   &ScaleColor(
    Scale Alpha=&FloatFromCurve(FloatCurve=DataInterface() {propertyValues {Curve {keys [<at 0 1>,<at 0.1864 1 RCIM_Constant>,<at 0.1997 -0.0006 RCIM_Constant>,<at 0.3026 1 RCIM_Constant>,<at 0.5431 1 RCIM_Constant>,<at 0.5634 0 RCIM_Constant>,<at 0.6714 1 RCIM_Constant>,<at 1 0>]}}})
    ScaleA=true
    ScaleRGB=true
   ) {module=ScaleColor001}
  ]
  rendererClasses [&NiagaraMeshRendererProperties]
  renderers [{rendererClass=&NiagaraMeshRendererProperties, rendererIndex=0}]
 }
 Smoke {
  emitterUpdateScript [
   &EmitterScalability(
    Module.bApplySpawnCountScale=false
    Module.bApplySpawnCountScaleByDistanceFraction=false
    Module.bUseMaxDistance=false
    Module.bUseMinDistance=false
   ) {enabled=false}
   &EmitterLifeCycle(NextLoopDuration=0) {enabled=false}
   &EmitterState(
    Enable Distance Culling=true
    Max Distance Response=::KillImmediately
    MaxDistance=7000
    Module.bUseMaxDistance=true
    Module.bUseMinDistance=false
    Scalability Mode=::NewEnumerator1
   )
   &SpawnBurst_Instantaneous(Spawn Count=&UniformRangedInt(Minimum=3, Maximum=5), Use Spawn Probability=false) {enabled=false}
  ]
  particleSpawnScript [
   &InitializeParticle(
    Color=&MakeLinearColorFromVectorAndFloat(
     "Vector (RGB)"=&MakeVectorFromLinearColorRGB(Color=<rgba 1 0.6432 0.6133 0.4>)
     "Float (Alpha)"=&UniformRangedFloat(Minimum=0.05, Maximum=0.1)
    )
    Lifetime=&UniformRangedFloat(Minimum=0.33, Maximum=0.44)
    Module.Write Color=true
    Module.Write Lifetime=true
    Module.Write Mass=true
    Module.Write Material Random=true
    Module.Write Position=false
    Module.Write Scale=false
    Module.Write SpriteRotation=true
    Module.Write SpriteSize=true
    Position=&SimulationPosition()
    Sprite Rotation=&NormalizedAngleToDegrees(NormalizedAngle=&UniformRangedFloat(Minimum=0, Maximum=1))
    Sprite Size=&Vector2DFromFloat(Value=&Multiply_Float(A=&UniformRangedFloat(Minimum=45, Maximum=60), B=0.2))
   )
   /SetVariables(Particles.InitSpriteRotRate=&UniformRangedFloat(Minimum=-3, Maximum=3)) {isSetParameters=true}
   &AddVelocityInCone(
    Cone Angle=0
    Cone Axis=<0 -1 0>
    Module.Use Velocity Falloff On Cone Axis=false
    Velocity Strength=&UniformRangedFloat(Minimum=30, Maximum=55)
   )
   &CameraOffset(Camera Offset Amount=&UniformRangedFloat(Minimum=0, Maximum=10))
   &SystemLocation(MaskSpawnGroup=false, Offset=<4 -2 0>)
  ]
  particleUpdateScript [
   &ParticleState()
   &SpriteRotationRate(Rotation Rate=&Multiply_Float(
     A=$Particles.InitSpriteRotRate
     B=&FloatFromCurve(FloatCurve=DataInterface() {propertyValues {Curve {keys [<at -0.0046 1.0113 RCIM_Cubic>,<at 1.0011 0.09 RCIM_Cubic>]}}})
    ))
   /Niagara/Modules/Update/SubUV/SubUVAnimation(Number Of Frames=64, SubUV Lookup Index=&Multiply_Float(A=$Particles.Age, B=0.66))
   &Drag(Drag=2, Use Linear Drag=true, Use Rotational Drag=false)
   &GravityForce(Gravity=<0 0 -100>)
   &SolveForcesAndVelocity(Clamp Velocity=false, Limit Acceleration=false)
   &Color(
    Color=$Particles.Initial.Color
    Scale Alpha=&FloatFromCurve(FloatCurve=DataInterface() {propertyValues {Curve {keys [<at 0 0.5 RCIM_Cubic>,<at 1.0011 0.002 RCIM_Cubic>]}}})
   )
   &ScaleSpriteSize(Scale Factor=&Vector2DFromFloat(Value=&FloatFromCurve(FloatCurve=DataInterface() {propertyValues {Curve {keys [
          {arriveTangent=2.9149, interpMode=RCIM_Cubic, leaveTangent=2.9149, tangentMode=RCTM_User, time=-0.003, value=0}
          {arriveTangent=0.9127, interpMode=RCIM_Cubic, leaveTangent=0.9127, tangentMode=RCTM_User, time=0.23, value=0.63}
          <at 1 1>
         ]}}})))
   /Game/Effects/Niagara/Modules/Materials/MaterialSmoke(
    Erode:NiagaraFloat=&FloatFromCurve(FloatCurve=DataInterface() {propertyValues {
       bOptimizeLUT=false
       bUseLUT=false
       Curve {keys [
         {arriveTangent=-0.0652, interpMode=RCIM_Cubic, leaveTangent=-0.0652, tangentMode=RCTM_User, time=0.11, value=0.97}
         {arriveTangent=-5.7797, interpMode=RCIM_Cubic, leaveTangent=-5.7797, tangentMode=RCTM_User, time=0.39, value=0.66}
         {arriveTangent=-0.342, interpMode=RCIM_Cubic, leaveTangent=-0.342, tangentMode=RCTM_User, time=0.9954, value=-0.74}
        ]}
      }})
    Shading:NiagaraFloat=&FloatFromCurve(FloatCurve=DataInterface() {propertyValues {
       bOptimizeLUT=false
       bUseLUT=false
       Curve {keys [
         <at 0.07 0.5 RCIM_Cubic>
         {arriveTangent=1.2205, interpMode=RCIM_Cubic, leaveTangent=1.2205, time=0.17, value=1.02}
         {interpMode=RCIM_Cubic, tangentMode=RCTM_User, time=0.3895, value=0.89}
         <at 0.8 0 RCIM_Cubic>
        ]}
      }})
   )
  ]
  rendererClasses [&NiagaraSpriteRendererProperties]
  renderers [{rendererClass=&NiagaraSpriteRendererProperties}]
 }
 left_eye_Sparks {
  emitterUpdateScript [&EmitterState(
    Life Cycle Mode=::NewEnumerator1
    Loop Delay=&UniformRangedFloat(Minimum=0, Maximum=2)
    Loop Duration=&UniformRangedFloat(Minimum=0.1, Maximum=0.1)
    UseLoopDelay=false
   )]
  particleSpawnScript [
   &InitializeParticle(
    Color=<rgba 0.0906 0.0496 1 1>
    Lifetime=&UniformRangedFloat(Minimum=0.2, Maximum=0.5)
    Module.Write Color=true
    Module.Write Lifetime=true
    Module.Write Mass=true
    Module.Write Material Random=true
    Module.Write Position=true
    Module.Write Scale=false
    Module.Write SpriteRotation=false
    Module.Write SpriteSize=true
    Position=&SimulationPosition()
    Sprite Size=&UniformRangedVector2D(Minimum=<1 3.5>, Maximum=<1 2>)
   )
   &AddVelocityInCone(
    Cone Angle=70
    Cone Axis=<0 -1 0.05>
    Module.Use Velocity Falloff On Cone Axis=false
    Velocity Strength=&UniformRangedFloat(Minimum=55, Maximum=55)
   )
   &SphereLocation(
    Fixed Random Seed=true
    MaskSpawnGroup=false
    Module.Override Randomness=true
    Module.Override Seed=true
    Offset=<3 -5 0.5>
    Random Seed=1
    Randomness Mode:ENiagaraRandomnessMode=::NewEnumerator1
    Sphere Radius=0.5
    Surface Only=true
   )
  ]
  particleUpdateScript [
   &ParticleState()
   &AccelerationForce(Acceleration=<0 0 -300>)
   &Drag(Drag=3, Use Linear Drag=true, Use Rotational Drag=false)
   &SolveForcesAndVelocity(Clamp Velocity=false, Limit Acceleration=false)
   &ScaleColor(
    Scale Alpha=10
    Scale RGB=&MakeVectorFromLinearColorRGB(Color=&ColorFromCurve(ColorCurve=DataInterface() {propertyValues {
        AlphaCurve {keys [<at 0 1 RCIM_Cubic>]}
        BlueCurve {keys [<at 0 100 RCIM_Cubic>,<at 0.7114 10 RCIM_Cubic>]}
        GreenCurve {keys [<at 0 100 RCIM_Cubic>,<at 0.7114 10 RCIM_Cubic>]}
        RedCurve {keys [<at 0 100 RCIM_Cubic>,<at 0.7114 10 RCIM_Cubic>]}
       }}))
    ScaleA=true
    ScaleRGB=true
   )
   &ScaleSpriteSize(Scale Factor=&Vector2DFromFloat(Value=&FloatFromCurve(
      FloatCurve=DataInterface() {propertyValues {Curve {keys [<at 0.0012 0.0032 RCIM_Cubic>, {arriveTangent=-2.1574, leaveTangent=-2.1574, time=0.0733, value=1}, <at 1 0>]}}}
      Scale Curve=0.8
     )))
  ]
  rendererClasses [&NiagaraSpriteRendererProperties]
  renderers [{rendererClass=&NiagaraSpriteRendererProperties}]
 }
 left_eye_Sparks001 {
  emitterUpdateScript [&EmitterState(
    Life Cycle Mode=::NewEnumerator1
    Loop Delay=&UniformRangedFloat(Minimum=0, Maximum=2)
    Loop Duration=&UniformRangedFloat(Minimum=0.1, Maximum=0.1)
    UseLoopDelay=false
   )]
  particleSpawnScript [
   &InitializeParticle(
    Color=<rgba 0.0906 0.0496 1 1>
    Lifetime=&UniformRangedFloat(Minimum=0.2, Maximum=0.5)
    Module.Write Color=true
    Module.Write Lifetime=true
    Module.Write Mass=true
    Module.Write Material Random=true
    Module.Write Position=true
    Module.Write Scale=false
    Module.Write SpriteRotation=false
    Module.Write SpriteSize=true
    Position=&SimulationPosition()
    Sprite Size=&UniformRangedVector2D(Minimum=<1 3.5>, Maximum=<1 2>)
   )
   &AddVelocityInCone(
    Cone Angle=70
    Cone Axis=<0 0 0.05>
    Module.Use Velocity Falloff On Cone Axis=false
    Velocity Strength=&UniformRangedFloat(Minimum=55, Maximum=55)
   )
   &SphereLocation(
    Fixed Random Seed=true
    MaskSpawnGroup=false
    Module.Override Randomness=true
    Module.Override Seed=true
    Offset=<-5 2 -19>
    Random Seed=1
    Randomness Mode:ENiagaraRandomnessMode=::NewEnumerator1
    Sphere Radius=0.5
    Surface Only=true
   )
  ]
  particleUpdateScript [
   &ParticleState()
   &AccelerationForce(Acceleration=<0 0 -300>)
   &Drag(Drag=3, Use Linear Drag=true, Use Rotational Drag=false)
   &SolveForcesAndVelocity(Clamp Velocity=false, Limit Acceleration=false)
   &ScaleColor(
    Scale Alpha=10
    Scale RGB=&MakeVectorFromLinearColorRGB(Color=&ColorFromCurve(ColorCurve=DataInterface() {propertyValues {
        AlphaCurve {keys [<at 0 1 RCIM_Cubic>]}
        BlueCurve {keys [<at 0 100 RCIM_Cubic>,<at 0.7114 10 RCIM_Cubic>]}
        GreenCurve {keys [<at 0 100 RCIM_Cubic>,<at 0.7114 10 RCIM_Cubic>]}
        RedCurve {keys [<at 0 100 RCIM_Cubic>,<at 0.7114 10 RCIM_Cubic>]}
       }}))
    ScaleA=true
    ScaleRGB=true
   )
   &ScaleSpriteSize(Scale Factor=&Vector2DFromFloat(Value=&FloatFromCurve(
      FloatCurve=DataInterface() {propertyValues {Curve {keys [<at 0.0012 0.0032 RCIM_Cubic>, {arriveTangent=-2.1574, leaveTangent=-2.1574, time=0.0733, value=1}, <at 1 0>]}}}
      Scale Curve=0.8
     )))
  ]
  rendererClasses [&NiagaraSpriteRendererProperties]
  renderers [{rendererClass=&NiagaraSpriteRendererProperties}]
 }
}
summary {
 systemName=NS_Backpack_TyphoonFemale_Loop
 emitters [
  {emitterName=left_eye_Sparks, rendererClasses [&NiagaraSpriteRendererProperties]}
  {emitterName=Marks_Lightning_Loop, rendererClasses [&NiagaraMeshRendererProperties]}
  {
   emitterName=Left_eye_lightning001
   rendererClasses [&NiagaraLightRendererProperties, &NiagaraMeshRendererProperties]
  }
  {
   emitterName=Left_eye_lightning002
   rendererClasses [&NiagaraLightRendererProperties, &NiagaraMeshRendererProperties]
  }
  {
   bEnabled=false
   emitterName=Left_eye_lightning
   rendererClasses [&NiagaraLightRendererProperties, &NiagaraMeshRendererProperties]
  }
  {
   emitterName=Left_eye_lightning003
   rendererClasses [&NiagaraLightRendererProperties, &NiagaraMeshRendererProperties]
  }
  {
   emitterName=Left_eye_lightning004
   rendererClasses [&NiagaraLightRendererProperties, &NiagaraMeshRendererProperties]
  }
  {emitterName=left_eye_Sparks001, rendererClasses [&NiagaraSpriteRendererProperties]}
  {emitterName=Smoke, rendererClasses [&NiagaraSpriteRendererProperties]}
 ]
 userVariables [User.Weapon {
   type=NiagaraDataInterfaceSkeletalMesh
   defaultValue {
    struct=NiagaraExt_VariableValue_DataInterface
    value {dataInterface=&self:NiagaraDataInterfaceSkeletalMesh_0}
   }
  }]
}
